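# Physics asset for the FirstClass cosmetic backpack, modeling a small figure with head, arms and legs for secondary motion.
# Capsule bodies form chains from body through tiny base anchors to dyn bones, with elbows and knees ending the limbs.
# Base-to-body joints leave all angular axes free, while dyn joints have limited swing and near-zero twist, tighter at head, elbows and knees than at shoulders and hips.
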PhysicsAsset /BRCosmetics/Accessories/FORT_Backpacks/Backpack_M_MED_FirstClass/Meshes/M_MED_FirstClass_Pack_Physics
bodies {
 base_arm_l [{shapeType=Capsule, length=0.5, radius=0.5, shapeName=base_arm_l_capsule}]
 base_arm_r [{shapeType=Capsule, length=0.5, radius=0.5, shapeName=base_arm_r_capsule}]
 base_head [{shapeType=Capsule, length=0.5, radius=0.5, shapeName=base_head_capsule}]
 base_leg_l [{shapeType=Capsule, length=0.5, radius=0.5, shapeName=base_leg_l_capsule}]
 base_leg_r [{shapeType=Capsule, length=0.5, radius=0.5, shapeName=base_leg_r_capsule}]
 body [{shapeType=Capsule, length=1, radius=3.5, shapeName=body_capsule, center=<1.5 0 0>, rotation=<rot 90 0 0>}]
 dyn_arm_l [{shapeType=Capsule, length=1.6433, radius=1, shapeName=dyn_arm_l_capsule, center=<0.8165 -0.15 0>, rotation=<rot 90 49.9625 49.9625>}]
 dyn_arm_r [{shapeType=Capsule, length=1.6638, radius=1, shapeName=dyn_arm_r_capsule, center=<0.8165 0.15 0>, rotation=<rot 90 0 0>}]
 dyn_elbow_l [{shapeType=Capsule, length=4, radius=2.2, shapeName=dyn_elbow_l_capsule, center=<2.2 0 0.7>, rotation=<rot 90 10 0>}]
 dyn_elbow_r [{shapeType=Capsule, length=4, radius=2.2, shapeName=dyn_elbow_r_capsule, center=<2.2 0 0.7>, rotation=<rot 90 -10 0>}]
 dyn_head [{shapeType=Capsule, length=4, radius=9.4, shapeName=dyn_head_capsule, center=<10.3752 -2.1331 0>, rotation=<rot 90 0 0>}]
 dyn_knee_l [{shapeType=Capsule, length=4, radius=2.05, shapeName=dyn_knee_l_capsule, center=<2.5 1.3 0.3>, rotation=<rot 5.0904 -10.1001 81.2435>}]
 dyn_knee_r [{shapeType=Capsule, length=4, radius=2.05, shapeName=dyn_knee_r_capsule, center=<2.5 1.3 -0.45>, rotation=<rot 0.8744 -10.3433 105.3695>}]
 dyn_leg_l [{shapeType=Capsule, length=1, radius=2.1, shapeName=dyn_leg_l_capsule, rotation=<rot 0 0 90>}]
 dyn_leg_r [{shapeType=Capsule, length=1, radius=2.1, shapeName=dyn_leg_r_capsule, rotation=<rot 0 0 90>}]
}
constraints [
 {bone1Name=dyn_arm_l, bone2Name=base_arm_l, swing1LimitDegrees=25, swing2LimitDegrees=30, twistLimitDegrees=1}
 {bone1Name=dyn_elbow_l, bone2Name=dyn_arm_l, swing1LimitDegrees=10, swing2LimitDegrees=20, twistLimitDegrees=1}
 {bone1Name=dyn_arm_r, bone2Name=base_arm_r, swing1LimitDegrees=25, swing2LimitDegrees=30, twistLimitDegrees=1}
 {bone1Name=dyn_elbow_r, bone2Name=dyn_arm_r, swing1LimitDegrees=10, swing2LimitDegrees=20, twistLimitDegrees=1}
 {bone1Name=dyn_head, bone2Name=base_head, swing1LimitDegrees=5, swing2LimitDegrees=10, twistLimitDegrees=1}
 {bone1Name=dyn_leg_l, bone2Name=base_leg_l, swing1LimitDegrees=30, swing2LimitDegrees=15, twistLimitDegrees=1}
 {bone1Name=dyn_knee_l, bone2Name=dyn_leg_l, swing1LimitDegrees=10, swing2LimitDegrees=5, twistLimitDegrees=1}
 {bone1Name=dyn_leg_r, bone2Name=base_leg_r, swing1LimitDegrees=30, swing2LimitDegrees=15, twistLimitDegrees=1}
 {bone1Name=dyn_knee_r, bone2Name=dyn_leg_r, swing1LimitDegrees=10, swing2LimitDegrees=5, twistLimitDegrees=1}
 {
  bone1Name=base_arm_r
  bone2Name=body
  swing1LimitDegrees=45
  swing1Motion=Free
  swing2LimitDegrees=45
  swing2Motion=Free
  twistLimitDegrees=45
  twistMotion=Free
 }
 {
  bone1Name=base_arm_l
  bone2Name=body
  swing1LimitDegrees=45
  swing1Motion=Free
  swing2LimitDegrees=45
  swing2Motion=Free
  twistLimitDegrees=45
  twistMotion=Free
 }
 {
  bone1Name=base_head
  bone2Name=body
  swing1LimitDegrees=45
  swing1Motion=Free
  swing2LimitDegrees=45
  swing2Motion=Free
  twistLimitDegrees=45
  twistMotion=Free
 }
 {
  bone1Name=base_leg_r
  bone2Name=body
  swing1LimitDegrees=45
  swing1Motion=Free
  swing2LimitDegrees=45
  swing2Motion=Free
  twistLimitDegrees=45
  twistMotion=Free
 }
 {
  bone1Name=base_leg_l
  bone2Name=body
  swing1LimitDegrees=45
  swing1Motion=Free
  swing2LimitDegrees=45
  swing2Motion=Free
  twistLimitDegrees=45
  twistMotion=Free
 }
]
